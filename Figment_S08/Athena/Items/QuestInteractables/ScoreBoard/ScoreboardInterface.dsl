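; Blueprint interface for a quest-interactable scoreboard in Fortnite's Figment Season 8 Athena content.
; It declares three empty functions that implementers provide.
; Startup takes an owning player actor, MoveScore takes a score increment plus the contributing Athena player controllers, and EndGame takes no parameters.
; MoveScore and EndGame have return nodes without outputs, while Startup has none.
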(fn EndGame
  (return))

(fn MoveScore
  (params (ScoreIncrement :type Integer) (ContributingPlayers :type Array%20of%20Fort%20Player%20Controller%20Athena%20Object%20References))
  (return))

(fn Startup
  (params (OwningPlayer :type Actor%20Object%20Reference)))
